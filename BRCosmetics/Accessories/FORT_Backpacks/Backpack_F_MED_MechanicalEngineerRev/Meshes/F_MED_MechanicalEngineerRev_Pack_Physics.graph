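# Physics asset for the Mechanical Engineer Rev backpack cosmetic, giving its bag, two swords and two hooks physics bodies.
# Each bone has one capsule, with long thin ones on the swords and tiny ones on root and pack.
# Constraints chain root, pack and bag with free angular motion, then branch to the left and right swords.
# Each hook hangs off its sword with one free swing axis, giving hinge-like motion.
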PhysicsAsset /BRCosmetics/Accessories/FORT_Backpacks/Backpack_F_MED_MechanicalEngineerRev/Meshes/F_MED_MechanicalEngineerRev_Pack_Physics
bodies {
 bag [{shapeType=Capsule, length=5.5053, radius=7.2579, shapeName=bag_capsule, center=<1.0861 -1.6967 1.1999>, rotation=<rot 6.9294 122.6044 -90>}]
 dyn_hook_l [{shapeType=Capsule, length=2.3333, radius=1, shapeName=dyn_hook_l_capsule, center=<1.9669 0.0218 0>, rotation=<rot 0 -94.9996 -90>}]
 dyn_hook_r [{shapeType=Capsule, length=7.5833, radius=0.5, shapeName=dyn_hook_l_capsule, center=<4.144 -0.1876 0>, rotation=<rot 0 -94.9996 -90>}]
 pack [{shapeType=Capsule, length=0.505, radius=0.505, shapeName=pack_capsule, rotation=<rot 0 -90 0>}]
 root [{shapeType=Capsule, length=0.505, radius=0.505, shapeName=root_capsule, rotation=<rot 0 -90 0>}]
 sword_l [{
   shapeType=Capsule
   length=37.8719
   radius=0.75
   shapeName=sword_l_capsule
   center=<-8.6213 -0.5597 0.0435>
   rotation=<rot 0 91.5132 -89.8985>
  }]
 sword_r [{shapeType=Capsule, length=40.3722, radius=0.75, shapeName=sword_r_capsule, center=<-8.621 -0.56 -0.0836>, rotation=<rot 0 91.5133 -89.8297>}]
}
constraints [
 {bone1Name=sword_l, bone2Name=bag, swing1LimitDegrees=45, swing2LimitDegrees=45, twistLimitDegrees=45}
 {bone1Name=sword_r, bone2Name=bag, swing1LimitDegrees=45, swing2LimitDegrees=45, twistLimitDegrees=45}
 {
  bone1Name=bag
  bone2Name=pack
  swing1LimitDegrees=45
  swing1Motion=Free
  swing2LimitDegrees=45
  swing2Motion=Free
  twistLimitDegrees=45
  twistMotion=Free
 }
 {
  bone1Name=pack
  bone2Name=root
  swing1LimitDegrees=45
  swing1Motion=Free
  swing2LimitDegrees=45
  swing2Motion=Free
  twistLimitDegrees=45
  twistMotion=Free
 }
 {bone1Name=dyn_hook_l, bone2Name=sword_l, swing1LimitDegrees=100, swing1Motion=Free, swing2LimitDegrees=1, twistLimitDegrees=1}
 {bone1Name=dyn_hook_r, bone2Name=sword_r, swing1LimitDegrees=100, swing1Motion=Free, swing2LimitDegrees=1, twistLimitDegrees=1}
]
