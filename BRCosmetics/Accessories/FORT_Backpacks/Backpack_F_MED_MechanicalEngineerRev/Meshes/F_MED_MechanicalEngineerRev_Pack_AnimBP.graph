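# Animation Blueprint for the Mechanical Engineer Rev backpack, which animates three floater joints and hides part of the pack while its matching pickaxe is equipped outside skydiving.
# Event graph sets sine bob offsets, spine-driven pack offsets, gravity and fall-dependent acceleration scale.
# These feed bone transforms and RigidBody into a cached pose, then two Anim Dynamics chains blended on grounded movement, one at zero alpha.
# Hiding sets a material scalar and calls the HideBackpack dispatcher.
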AnimBlueprint /BRCosmetics/Accessories/FORT_Backpacks/Backpack_F_MED_MechanicalEngineerRev/Meshes/F_MED_MechanicalEngineerRev_Pack_AnimBP
target_skeleton=&/F_MED_MechanicalEngineerRev_Pack_Skeleton
graphs [
 AnimGraph {nodes [
   n1 = Anim Dynamics - Bone: ambient_jnt_mid(ComponentPose=Local To Component(LocalPose=Use cached pose 'Translate Cache'()), Alpha=1, ExternalForce="0,0,0") {
    simulationSpace=RootRelative
    alphaBoolBlend {blendOption=Linear}
    boundBone {boneName=ambient_jnt_mid}
    physicsBodyDefinitions [{
      boundBone {boneName=ambient_jnt_mid}
      boxExtents=<10 10 10>
      constraintSetup {angularLimitsMax=<0 0 0>, angularLimitsMin=<0 0 0>, angularTarget=<0 0 0>}
      localJointOffset=<0 0 0>
     }]
    relativeSpaceBone {boneName=root}
   }
   n2 = Anim Dynamics - Bone: ambient_jnt_r(ComponentPose=@n1, Alpha=1, ExternalForce="0,0,0") {
    simulationSpace=RootRelative
    alphaBoolBlend {blendOption=Linear}
    boundBone {boneName=ambient_jnt_r}
    physicsBodyDefinitions [{
      boundBone {boneName=ambient_jnt_r}
      boxExtents=<10 10 10>
      constraintSetup {angularLimitsMax=<0 0 0>, angularLimitsMin=<0 0 0>, angularTarget=<0 0 0>}
      localJointOffset=<0 0 0>
     }]
    relativeSpaceBone {boneName=root}
   }
   n3 = Anim Dynamics - Bone: ambient_jnt_l(ComponentPose=@n2, Alpha=1, ExternalForce="0,0,0") {
    simulationSpace=RootRelative
    alphaBoolBlend {blendOption=Linear}
    boundBone {boneName=ambient_jnt_l}
    physicsBodyDefinitions [{
      boundBone {boneName=ambient_jnt_l}
      boxExtents=<10 10 10>
      constraintSetup {angularLimitsMax=<0 0 0>, angularLimitsMin=<0 0 0>, angularTarget=<0 0 0>}
      localJointOffset=<0 0 0>
     }]
    relativeSpaceBone {boneName=root}
   }
   n4 = Anim Dynamics - Bone: ambient_jnt_mid(ComponentPose=Local To Component(LocalPose=Use cached pose 'Translate Cache'()), Alpha=0, ExternalForce="0,0,0") {
    angularDampingOverride=0.9
    angularSpringConstant=200
    bGravityOverrideInSimSpace=true
    bLinearSpring=true
    bOverrideAngularDamping=true
    bOverrideLinearDamping=true
    gravityScale=0
    linearDampingOverride=0.9999
    linearSpringConstant=100
    simulationSpace=RootRelative
    alphaBoolBlend {blendOption=Linear}
    boundBone {boneName=ambient_jnt_mid}
    componentAppliedLinearAccClamp=<50000 50000 50000>
    componentLinearAccScale {x=0.1, z=0.1}
    componentLinearVelScale=<0.01 0.01 0.01>
    physicsBodyDefinitions [{
      sphereCollisionRadius=0
      boundBone {boneName=ambient_jnt_mid}
      boxExtents=<4 4 4>
      constraintSetup {angularLimitsMax=<0 0 0>, angularLimitsMin=<0 0 0>, angularTarget=<0 0 0>, linearAxesMax=<4.3 4.3 4.3>, linearAxesMin=<-0.5 -0.1 -0.5>}
      localJointOffset=<0 0 0>
     }]
    relativeSpaceBone {boneName=root}
   }
   n5 = Anim Dynamics - Bone: ambient_jnt_r(ComponentPose=@n4, Alpha=0, ExternalForce="0,0,0") {
    angularDampingOverride=0.9
    angularSpringConstant=200
    bGravityOverrideInSimSpace=true
    bLinearSpring=true
    bOverrideAngularDamping=true
    bOverrideLinearDamping=true
    gravityScale=0
    linearDampingOverride=0.9999
    linearSpringConstant=100
    simulationSpace=RootRelative
    alphaBoolBlend {blendOption=Linear}
    boundBone {boneName=ambient_jnt_r}
    componentAppliedLinearAccClamp=<30000 30000 30000>
    componentLinearAccScale {x=0.1, z=0.1}
    componentLinearVelScale=<0.01 0.2 0.01>
    physicsBodyDefinitions [{
      sphereCollisionRadius=0
      boundBone {boneName=ambient_jnt_r}
      boxExtents=<4 4 4>
      constraintSetup {angularLimitsMax=<0 0 0>, angularLimitsMin=<0 0 0>, angularTarget=<0 0 0>, linearAxesMax=<5 5 5>, linearAxesMin=<-0.5 -0.1 -0.5>}
      localJointOffset=<0 0 0>
     }]
    relativeSpaceBone {boneName=root}
   }
   n6 = Anim Dynamics - Bone: ambient_jnt_l(ComponentPose=@n5, Alpha=0, ExternalForce="0,0,0") {
    angularDampingOverride=0.9
    angularSpringConstant=200
    bGravityOverrideInSimSpace=true
    bLinearSpring=true
    bOverrideAngularDamping=true
    bOverrideLinearDamping=true
    gravityScale=0
    linearDampingOverride=0.9999
    linearSpringConstant=100
    simulationSpace=RootRelative
    alphaBoolBlend {blendOption=Linear}
    boundBone {boneName=ambient_jnt_l}
    componentAppliedLinearAccClamp=<50000 50000 50000>
    componentLinearAccScale=<0.03 0.01 0.03>
    componentLinearVelScale=<0.03 0.03 0.03>
    physicsBodyDefinitions [{
      sphereCollisionRadius=0
      boundBone {boneName=ambient_jnt_l}
      boxExtents=<4 4 4>
      constraintSetup {angularLimitsMax=<0 0 0>, angularLimitsMin=<0 0 0>, angularTarget=<0 0 0>, linearAxesMax=<3.5 3.5 3.5>, linearAxesMin=<-0.5 -0.5 -0.5>}
      localJointOffset=<0 0 0>
     }]
    relativeSpaceBone {boneName=root}
   }
   n7 = Get bIsOnGround() -> (bIsOnGround:bool)
   n8 = Get bIsMoving2D()
   Output Pose(Result=Blend Poses by bool(
     BlendPose_0=Component To Local(ComponentPose=@n3)
     BlendPose_1=Component To Local(ComponentPose=@n6)
     BlendTime_0=0.75
     BlendTime_1=0.75
     bActiveValue=AND Boolean(A=@n7, B=@n8)
    ) {blendType=Linear}) {layerGroup=None}
   n9 = "Transform (Modify) Bone - Bone: pack"(
    ComponentPose=Local To Component(LocalPose=Input Pose())
    Translation=Get backpack_translation()
    Rotation=Get backpack_rotate()
    Scale="1.2,1.2,1.2"
   ) {
    rotationMode=BMM_Additive
    rotationSpace=BCS_BoneSpace
    scaleMode=BMM_Additive
    scaleSpace=BCS_BoneSpace
    translationMode=BMM_Additive
    translationSpace=BCS_BoneSpace
    alphaBoolBlend {blendOption=Linear}
    boneToModify {boneName=pack}
    scale=<1.2 1.2 1.2>
   }
   n10 = "Transform (Modify) Bone - Bone: ambient_jnt_mid"(
    ComponentPose=@n9
    Alpha=1
    Translation="Get Translate Floaters (Top)"() -> ("Translate Floaters (Top)":Vector)
    Rotation="0,0,0"
    Scale="1,1,1"
   ) {translationMode=BMM_Additive, translationSpace=BCS_BoneSpace, alphaBoolBlend {blendOption=Linear}, boneToModify {boneName=ambient_jnt_mid}}
   n11 = "Transform (Modify) Bone - Bone: ambient_jnt_r"(
    ComponentPose=@n10
    Alpha=1
    Translation="Get Translate Floaters (Bottom_r)"() -> ("Translate Floaters (Bottom_r)":Vector)
    Rotation="0,0,0"
    Scale="1,1,1"
   ) {translationMode=BMM_Additive, translationSpace=BCS_BoneSpace, alphaBoolBlend {blendOption=Linear}, boneToModify {boneName=ambient_jnt_r}}
   n12 = "Transform (Modify) Bone - Bone: ambient_jnt_l"(
    ComponentPose=@n11
    Alpha=1
    Translation="Get Translate Floaters (Bottom_l)"() -> ("Translate Floaters (Bottom_l)":Vector)
    Rotation="0,0,0"
    Scale="1,1,1"
   ) {translationMode=BMM_Additive, translationSpace=BCS_BoneSpace, alphaBoolBlend {blendOption=Linear}, boneToModify {boneName=ambient_jnt_l}}
   n13 = Get LinearAccScale() -> (LinearAccScale:Vector)
   n14 = RigidBody!(
    ComponentPose=@n12
    Alpha
    OverrideWorldGravity:Vector=Get GravityOverride()
    ExternalForce
    ComponentLinearAccScale:Vector=@n13
    ComponentLinearVelScale:Vector=@n13
   ) -> (Pose) {
    bClampLinearTranslationLimitToRefPose=true
    bOverrideWorldGravity=true
    overridePhysicsAsset=&/F_MED_MechanicalEngineerRev_Pack_Physics
    simulationSpace=BaseBoneSpace
    alphaBoolBlend {blendOption=Linear}
    baseBoneRef {boneName=root}
    componentAppliedLinearAccClamp=<3000 3000 3000>
    componentLinearAccScale=<0.4 0.3 0.6>
    componentLinearVelScale=<0.4 0.3 0.3>
    overrideWorldGravity {z=-980}
    simSpaceSettings {
     maxAngularAcceleration=1000
     maxAngularVelocity=3
     maxLinearAcceleration=700
     maxLinearVelocity=500
     velocityScaleZ=0
     externalLinearDragV=<0.2 0.2>
    }
   }
   Save cached pose 'Translate Cache'(Pose=Component To Local(ComponentPose=@n14)) {cacheName=Translate Cache}
  ]}
 EventGraph {nodes [
   Event BlueprintUpdateAnimation()
   > n1 = Sequence!(execute) -> (then_0, then_1, then_2:Exec, then_3:Exec, then_4:Exec, then_5:Exec)
   > n2 = PlayerMovement(AnimInstance=Self-Reference(), XDotProductCoefficient=1, YDotProductCoefficient=1, ZDotProductCoefficient=1)
   > Set Y Movement(Y Movement:double=@n2.YMovement) -> (Output_Get:double)
   > n3 = Branch(Condition=Get bIsFalling())
   > Set LinearAccScale(LinearAccScale:Vector="0.4,0.3,1.09") -> (Output_Get:Vector)
   Set LinearAccScale(execute=@n3.else, LinearAccScale="0.4,0.3,0.6")
   n4 = Gravity Override State Function DEPRECATED(
    execute=@n1.then_1
    CharacterPartAnimInstance=Self-Reference()
    GameTimeInSeconds=GetGameTimeInSeconds()
    GravityOverrideParams=Get Gravity Override Params()
   )
   > Set GravityOverride(GravityOverride=@n4)
   n5 = "Sin (Radians)"(A=float * float(A=GetRealTimeSeconds(), B=2.3))
   "Set Translate Floaters (Top)"(execute=@n1.then_2, "Translate Floaters (Top)":Vector=MakeVector(Y=0, Z=float / float(A=@n5, B=-1.3) {is=title K2Node_CallFunction})) -> (Output_Get:Vector)
   n6 = GetRealTimeSeconds()
   n7 = float / float(A="Sin (Radians)"(A=float * float(A=@n6, B=2.3)), B=1.3)
   "Set Translate Floaters (Bottom_r)"(execute=@n1.then_3, "Translate Floaters (Bottom_r)":Vector=MakeVector(Y=@n7)) -> (Output_Get:Vector)
   "Set Translate Floaters (Bottom_l)"(execute=@n1.then_4, "Translate Floaters (Bottom_l)":Vector=MakeVector(Y=float * float(A=@n7, B=-1))) -> (Output_Get:Vector)
   n8 = GetOwnerSkeletalMeshComponent()
   n9 = BreakRotator(InRot=BreakTransform(InTransform=GetDeltaTransformFromRefPose(self=@n8, BoneName=spine_05, BaseName=spine_01)).Rotation)
   Set backpack_rotate(execute=@n1.then_5, backpack_rotate=MakeRotator(Roll=MapRangeClamped(Value=@n9.Yaw, InRangeA=0, InRangeB=-30, OutRangeB=15)))
   > Set backpack_translation(backpack_translation=MakeVector(Y=MapRangeClamped(Value=@n9.Yaw, InRangeA=0, InRangeB=-30, OutRangeA=3, OutRangeB=-8), Z=-2))
   n10 = Event OnWeaponChanged()
   n11 = Get Dynamic Material Instance(execute=[@n10.then, @n17.then], SkeletalMeshComponent=GetPartSkeletalMeshComponent())
   n12 = PickaxeItemDefinitionHasAnyMetaTag(Tags=Make Array("[0]"=(TagName="Cosmetics.MetaTags.MechanicalEngineerRev")), bRequiresEquipped=true)
   > n13 = Branch(Condition=@n12)
   > n14 = Branch(Condition=Get bIsSkyDiving())
   SetScalarParameterValue(execute=[@n15.then, @n14.then, @n16.else, @n13.else], self=@n11, ParameterName=Hide Element 01, Value=0)
   > Call Hide Backpack!(execute:Exec, self:F MED Mechanical Engineer Rev Pack Anim BP Object Reference, blsHidden:bool) -> (then:Exec) {is=title K2Node_CallDelegate}
   SetScalarParameterValue(execute=[@n14.else, @n16.then], self=@n11, ParameterName=Hide Element 01, Value=1)
   > Call Hide Backpack(blsHidden=true)
   n15 = Event OnBeginSkydiving()
   Event OnEndSkydiving()
   > n16 = Branch(Condition=@n12)
   n17 = Event OnCharacterCustomizationLoaded()
   TryGetPawnOwner()
  ]}
 HideBackpack {nodes [HideBackpack!() -> (then, blsHidden:bool=~) {is=title K2Node_FunctionEntry}]}
]
variables [
 backpack_rotate {type=Rotator}
 backpack_translation {type=Vector}
 Gravity Override Params {type=GravityOverrideParamsStruct}
 GravityOverride {type=Vector}
 HideBackpack {default=()}
 Is Jumping {type=boolean}
 IsSkydiveFloating {type=boolean}
 LinearAccScale {type=Vector}
 "Translate Floaters (Bottom_l)" {type=Vector}
 "Translate Floaters (Bottom_r)" {type=Vector}
 "Translate Floaters (Top)" {type=Vector}
 Y Movement {type=number}
]
